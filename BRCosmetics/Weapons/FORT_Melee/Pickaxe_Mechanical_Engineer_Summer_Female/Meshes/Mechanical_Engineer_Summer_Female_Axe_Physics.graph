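# Physics asset for the Mechanical Engineer Summer Female pickaxe, a two-handed melee cosmetic.
# It gives the handle bone and the dyn_jaw bone one capsule body each, with the jaw capsule offset and rotated from its bone.
# A single constraint joins the jaw to the handle with both swings and twist locked, so the jaw cannot rotate relative to the handle in simulation.
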PhysicsAsset /BRCosmetics/Weapons/FORT_Melee/Pickaxe_Mechanical_Engineer_Summer_Female/Meshes/Mechanical_Engineer_Summer_Female_Axe_Physics
bodies {
 Melee_TwoHanded_Handle [{shapeType=Capsule, length=60, radius=3}]
 dyn_jaw [{shapeType=Capsule, length=55, radius=5, center=<12 0 0>, rotation=<rot 90 0 0>}]
}
constraints [{
  bone1Name=dyn_jaw
  bone2Name=Melee_TwoHanded_Handle
  swing1LimitDegrees=45
  swing1Motion=Locked
  swing2LimitDegrees=45
  swing2Motion=Locked
  twistLimitDegrees=45
  twistMotion=Locked
 }]
