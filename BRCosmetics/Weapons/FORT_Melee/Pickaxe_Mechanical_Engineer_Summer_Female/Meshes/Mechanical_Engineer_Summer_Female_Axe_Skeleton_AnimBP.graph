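# Animation Blueprint for the Mechanical Engineer Summer Female pickaxe, adding secondary physics motion and a jaw-linked turning nut.
# A RigidBody node simulates in base-bone space, then AnimDynamics swings dyn_hinge within a narrow angular limit.
# Each update maps dyn_jaw's ref-pose offset against the handle to a clamped roll in Nut Rotate.
# Two additive bone transforms turn dyn_nut by a fixed offset, then by that roll.
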AnimBlueprint /BRCosmetics/Weapons/FORT_Melee/Pickaxe_Mechanical_Engineer_Summer_Female/Meshes/Mechanical_Engineer_Summer_Female_Axe_Skeleton_AnimBP
target_skeleton=&/Mechanical_Engineer_Summer_Female_Axe_Skeleton
graphs [
 AnimGraph {nodes [
   n1 = RigidBody(ComponentPose=Local To Component(LocalPose=Input Pose())) {
    bClampLinearTranslationLimitToRefPose=true
    bOverrideWorldGravity=true
    overridePhysicsAsset=&/Mechanical_Engineer_Summer_Female_Axe_Physics
    simulationSpace=BaseBoneSpace
    alphaBoolBlend {blendOption=Linear}
    baseBoneRef {boneName=b_AR_Root}
    componentAppliedLinearAccClamp=<3000 3000 3000>
    componentLinearAccScale=<0.6 0.3 0.3>
    componentLinearVelScale=<0.6 0.3 0.4>
    overrideWorldGravity {z=-980}
   }
   n2 = Anim Dynamics - Bone: dyn_hinge(ComponentPose=@n1, Alpha=1, ExternalForce="0,0,0") {
    angularDampingOverride=0.97
    bGravityOverrideInSimSpace=true
    bOverrideAngularDamping=true
    bUseGravityOverride=true
    alphaBoolBlend {blendOption=Linear}
    boundBone {boneName=dyn_hinge}
    componentAppliedLinearAccClamp=<3000 3000 3000>
    componentLinearAccScale=<0.3 0.3 0.4>
    componentLinearVelScale=<0.3 0.3 0.4>
    gravityOverride {z=-980}
    physicsBodyDefinitions [{
      sphereCollisionRadius=0
      boundBone {boneName=dyn_hinge}
      boxExtents=<50 100 100>
      constraintSetup {angularLimitsMax=<0 2 0>, angularLimitsMin=<0 0 0>, angularTarget=<0 0 0>}
      localJointOffset=<25 0 0>
     }]
   }
   n3 = "Transform (Modify) Bone - Bone: dyn_nut"(ComponentPose=@n2, Alpha=1, Translation="0,0,0", Rotation="0,0,27", Scale="1,1,1") {
    rotationMode=BMM_Additive
    rotationSpace=BCS_BoneSpace
    alphaBoolBlend {blendOption=Linear}
    boneToModify {boneName=dyn_nut}
    rotation {roll=27}
   }
   n4 = "Transform (Modify) Bone - Bone: dyn_nut"(ComponentPose=@n3, Alpha=1, Translation="0,0,0", Rotation=Get Nut Rotate() -> (Nut Rotate:Rotator), Scale="1,1,1") {rotationMode=BMM_Additive, rotationSpace=BCS_ParentBoneSpace, alphaBoolBlend {blendOption=Linear}, boneToModify {boneName=dyn_nut}}
   Output Pose(Result=Component To Local(ComponentPose=@n4)) {layerGroup=None}
  ]}
 EventGraph {nodes [
   Event BlueprintUpdateAnimation()
   n1 = GetOwningComponent()
   n2 = BreakVector(InVec=BreakTransform(InTransform=GetDeltaTransformFromRefPose(self=@n1, BoneName=dyn_jaw, BaseName=Melee_TwoHanded_Handle)))
   > Set Nut Rotate(Nut Rotate:Rotator=MakeRotator(Roll=MapRangeClamped(Value=@n2, InRangeA=-5, InRangeB=5, OutRangeA=-90, OutRangeB=90))) -> (Output_Get:Rotator)
   TryGetPawnOwner()
  ]}
]
variables [Nut Rotate {type=Rotator}]
